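# Child animation Blueprint for the Chameleon pet that pulses its body color in response to pet reaction tags.
# On an active tag change it calls the parent, then switches on the tag for rare, epic and legendary item pickups and Victory Royale.
# Each case animates the ChameleonColorShift parameter on the pet mesh's first dynamic material with its own color curve, with the victory pulse lasting longer.
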AnimBlueprint /BRCosmetics/Characters/Pets/Chameleon/B_ChameleonAnimBP_Child
target_skeleton=/Game/Characters/Pets/Chameleon/Meshes/Chameleon_Skeleton
graphs [EventGraph {nodes [
   n1 = Event TEMP_OnActiveTagChanged!() -> (OutputDelegate, then, NewTag:Gameplay Tag Structure) {is=title K2Node_Event}
   > Parent: TEMP_OnActiveTagChanged!(execute, NewTag:Gameplay Tag Structure=@n1.NewTag) -> (then) {is=title K2Node_CallParentFunction}
   > n2 = Switch on Gameplay Tag!(execute:Exec, Selection:Gameplay Tag Structure=@n1.NewTag) -> (
    Default:Exec
    Cosmetics.Pet.Response.Collect.RareItem:Exec
    Cosmetics.Pet.Response.Collect.EpicItem:Exec
    Cosmetics.Pet.Response.Collect.LegendaryItem:Exec
    Cosmetics.Pet.Response.Happy.VictoryRoyale:Exec
   ) {
    is=title GameplayTagsK2Node_SwitchGameplayTag
    pinTags [
     {tagName=Cosmetics.Pet.Response.Collect.RareItem}
     {tagName=Cosmetics.Pet.Response.Collect.EpicItem}
     {tagName=Cosmetics.Pet.Response.Collect.LegendaryItem}
     {tagName=Cosmetics.Pet.Response.Happy.VictoryRoyale}
    ]
   }
   n3 = Get OwnerPet() -> (OwnerPet:Fort Player Pet Object Reference)
   n4 = Cast To MaterialInstanceDynamic!(Object=GetMaterial!(
     self:Primitive Component Object Reference=Get PetMesh!(self:Fort Player Pet Object Reference=@n3) -> (PetMesh:Skeletal Mesh Component Budgeted Object Reference)
     ElementIndex:int=0
    ) -> (ReturnValue:Material Interface Object Reference)) -> (AsMaterial Instance Dynamic:Material Instance Dynamic Object Reference, bSuccess:bool=~) {is=title K2Node_DynamicCast}
   AddMIDAnimationColorCurve!(
    execute=@n2.Cosmetics.Pet.Response.Collect.EpicItem
    MID:Material Instance Dynamic Object Reference=@n4
    ParamName:Name=ChameleonColorShift
    ColorCurve:Curve Linear Color Object Reference=&/ChameleonHappyPulsePurple
    Duration:float=2
    bPlayReverse:bool=false
   ) -> (then)
   AddMIDAnimationColorCurve(
    execute=@n2.Cosmetics.Pet.Response.Collect.LegendaryItem
    MID=@n4
    ParamName=ChameleonColorShift
    ColorCurve=&/ChameleonHappyPulseGold
    Duration=2
    bPlayReverse=false
   )
   AddMIDAnimationColorCurve(
    execute=@n2.Cosmetics.Pet.Response.Collect.RareItem
    MID=@n4
    ParamName=ChameleonColorShift
    ColorCurve=&/ChameleonHappyPulseBlue
    Duration=2
    bPlayReverse=false
   )
   AddMIDAnimationColorCurve(
    execute=@n2.Cosmetics.Pet.Response.Happy.VictoryRoyale
    MID=@n4
    ParamName=ChameleonColorShift
    ColorCurve=&/ChameleonHappyPulseVictory
    Duration=5
    bPlayReverse=false
   )
  ]}]
variables []
